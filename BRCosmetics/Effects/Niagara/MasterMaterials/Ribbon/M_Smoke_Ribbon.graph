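# Translucent unlit master material for smoke trails on Niagara ribbons.
# A panning, noise-distorted UV drives a normal map and mask; the normal dotted with sky atmosphere light direction fakes lighting, blending two color parameters tinted by particle color.
# Cloud noise and gradients along the ribbon shape and fade the mask.
# A dynamic parameter erodes it, and a steep multiply sharpens the edge before particle alpha scales it.
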Material /BRCosmetics/Effects/Niagara/MasterMaterials/Ribbon/M_Smoke_Ribbon
graph {
 properties {
  bAutomaticallySetUsageInEditor=false
  bUsedWithNiagaraRibbons=true
  blendMode=BLEND_Translucent
  shadingModel=MSM_Unlit
  usedShadingModels=Unlit
 }
 nodes [
  n1 = VectorParameter() {parameterName=Color2, defaultValue=<rgba 0.5862 0.0828 0.6615 1>}
  n2 = VectorParameter() {parameterName=Color1, defaultValue=<rgba 1 0.329 0.7734 1>}
  n3 = TextureCoordinate() {coordinateIndex=1}
  n4 = Constant2Vector() {g=1, r=1.4}
  n5 = TextureSample(UVs=Multiply(A=TextureCoordinate(), B=Constant2Vector() {g=1, r=1.2})) {samplerType=SAMPLERTYPE_Normal, texture=/BRCosmetics/Characters/Player/Male/Large/Bodies/M_LRG_Clash_V/Textures/Simple_Perlin_N}
  n6 = Add(A=Panner(Coordinate=Multiply(A=@n3, B=@n4)) {speedX=0.3}, B=Multiply(A=Constant2Vector() {r=1}, B=Multiply(A=@n5.R) {constB=0.02}))
  n7 = TextureSampleParameter2D(UVs=@n6) {
   parameterName=TextureNormalMap
   samplerType=SAMPLERTYPE_Normal
   texture=/Game/Effects/Fort_Effects/Gameplay/SmokeDustExplosions/Textures/T_RibbonSmokeMasked_01
  }
  n8 = SkyAtmosphereLightDirection()
  n9 = Saturate(None=DotProduct(A=Transform(None=@n7.RGB), B=Normalize(VectorInput=@n8)))
  n10 = ParticleColor()
  MaterialFunctionCall(Emissive=Multiply(A=LinearInterpolate(A=@n1.RGB, B=@n2.RGB, Alpha=@n9), B=@n10.RGB)) {materialFunction=&MF_FX_EmissiveAdjustments}
  n11 = TextureCoordinate()
  n12 = Multiply(A=@n11, B=Constant2Vector() {g=4, r=1.5})
  n13 = ComponentMask(None=@n11) {r=true}
  n14 = OneMinus(None=@n13)
  n15 = Add(
   A=Multiply(A=TextureSample(UVs=@n12) {texture=&AtmosphericCloudNoise01}.R, B=@n14)
   B=SmoothStep(Value=@n14) {constMin=0.3}
  )
  n16 = Multiply(
   A=Multiply(
    A=TextureSampleParameter2D(UVs=@n6) {parameterName=TextureMask, texture=/Game/Effects/Fort_Effects/Gameplay/SmokeDustExplosions/Textures/T_RibbonSmokeMasked_04}.R
    B=Saturate(None=@n15)
   )
   B=SmoothStep(Value=@n13) {constMax=0.08, constMin=0.01}
  )
  n17 = SmoothStep(Value=@n13) {constMin=0.85}
  n18 = Multiply(A=Subtract(A=Subtract(A=@n16, B=@n17), B=DynamicParameter() {defaultValue=<rgba 0 1 1 1>}.Param1)) {constB=15}
  Multiply(A=@n10.A, B=Saturate(None=Subtract(A=@n18) {constB=0.15}))
  OneMinus(None=@n10.A)
 ]
}
